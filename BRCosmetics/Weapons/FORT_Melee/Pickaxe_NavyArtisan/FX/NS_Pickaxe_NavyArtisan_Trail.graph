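# Niagara swing trail for the NavyArtisan pickaxe, built from short-lived animation-driven ribbons.
# Trail and TrailGrid emitters use the shared FN anim-trail modules to spawn, initialize and update particles, each feeding two ribbon renderers.
# Both derive separate ColorBlue and ColorGrid tints from particle color and scale alpha by time of day.
# Trail fades along a curve, while TrailGrid lasts slightly longer and follows system-level fade variables.
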NiagaraSystem /BRCosmetics/Weapons/FORT_Melee/Pickaxe_NavyArtisan/FX/NS_Pickaxe_NavyArtisan_Trail
dependencies {
 usedDataInterfaces [
  &NiagaraDataInterfaceSkeletalMesh
  &NiagaraDataInterfaceCurve
  &NiagaraDataInterfaceSocketReader
 ]
 usedDynamicInputs [
  &Multiply_Vector4
  &MakeVector4FromVectorAndFloat
  &MergeHelper_FloatSwitch
  &FNTimeOfDayColor
  &ClampFloat
  &RemapRange
  &Add_Float
  &Subtract_Float
 ]
 usedModules [
  &self:SystemSpawnScript.SystemScriptSource.SystemScriptGraph.NiagaraNodeAssignment_0.SetVariables_#1
  &self:SystemSpawnScript.SystemScriptSource.SystemScriptGraph.NiagaraNodeAssignment_1.SetVariables_#2
  &SystemState
  &self:SystemSpawnScript.SystemScriptSource.SystemScriptGraph.NiagaraNodeAssignment_6.SetVariables_#3
  &self:SystemSpawnScript.SystemScriptSource.SystemScriptGraph.NiagaraNodeAssignment_5.SetVariables_#4
 ]
}
emitters {
 Trail {
  emitterUpdateScript [
   &EmitterState(Life Cycle Mode=::NewEnumerator1, Loop Behavior=::NewEnumerator1, Loop Duration=0.27, UseLoopDelay=true)
   &FN_SpawnAnimTrail()
  ]
  particleSpawnScript [
   &FN_InitializeAnimTrailBoneIndecies() {module=FN_InitializeAnimTrailBoneIndecies001}
   &FN_InitializeAnimTrail(AnimTrailWidth=0.8, Lifetime=0.2) {module=FN_InitializeAnimTrail001}
  ]
  particleUpdateScript [
   &FN_UpdateAnimTrail()
   &ScaleColor(
    Scale Alpha=&Multiply_Float(
     A=&FloatFromCurve(FloatCurve=DataInterface() {propertyValues {Curve {keys [
          <at 0 1 RCIM_Cubic>
          {interpMode=RCIM_Cubic, leaveTangent=-2.3646, tangentMode=RCTM_Break, time=0.0726, value=0.9843}
          <at 0.6689 0.0233>
          {arriveTangent=-0.2139, interpMode=RCIM_Cubic, tangentMode=RCTM_Break, time=1, value=0}
         ]}}})
     B=&FNTimeOfDayScalar(Night Scale=1, Morning Scale=0.9, Day Scale=0.8, Evening Scale=0.9)
    )
    ScaleA=true
    ScaleRGB=true
   )
   /SetVariables(
    Particles.ColorBlue:LinearColor=&MultiplyLinearColor(Linear Color=<rgba 1 1 1 4>, Scale Factor=&MakeVector4_FromLinearColor(LinearColor=$Particles.Color))
    Particles.ColorGrid:LinearColor=&MultiplyLinearColor(
     Linear Color=<rgba 0.9652 0.9846 1 0.2087>
     Scale Factor=&MakeVector4_FromLinearColor(LinearColor=$Particles.Color)
    )
   ) {isSetParameters=true}
  ]
  rendererClasses [&NiagaraRibbonRendererProperties]
  renderers [
   {rendererClass=&NiagaraRibbonRendererProperties}
   {rendererClass=&NiagaraRibbonRendererProperties, rendererIndex=1}
  ]
 }
 TrailGrid {
  emitterUpdateScript [
   &EmitterState(Life Cycle Mode=::NewEnumerator1, Loop Behavior=::NewEnumerator1, Loop Duration=0.27, UseLoopDelay=true)
   &FN_SpawnAnimTrail()
  ]
  particleSpawnScript [
   &FN_InitializeAnimTrailBoneIndecies() {module=FN_InitializeAnimTrailBoneIndecies001}
   &FN_InitializeAnimTrail(
    AnimTrailWidth=0.8
    Color=&MakeLinearColorFromVectorAndFloat(
     "Vector (RGB)"=&MakeVectorFromLinearColorRGB(Color=<rgba 1 1 1 1>)
     "Float (Alpha)"=&Multiply_Float(A=1, B=&Multiply_Float(A=$System.FadeSpawn, B=$System.FadeSpawnOut))
    )
    Lifetime=0.22
   ) {module=FN_InitializeAnimTrail001}
  ]
  particleUpdateScript [
   &FN_UpdateAnimTrail()
   &ScaleColor(
    Scale Alpha=&Multiply_Float(
     A=$System.FadeSystem
     B=&FNTimeOfDayScalar(Night Scale=1, Morning Scale=0.9, Day Scale=0.8, Evening Scale=0.9)
    )
    ScaleA=true
    ScaleRGB=true
   )
   /SetVariables(
    Particles.ColorBlue:LinearColor=&MultiplyLinearColor(
     Linear Color=<rgba 0.049 0.1195 0.313 3>
     Scale Factor=&MakeVector4_FromLinearColor(LinearColor=$Particles.Color)
    )
    Particles.ColorGrid:LinearColor=&MultiplyLinearColor(
     Linear Color=<rgba 0.9652 0.9846 1 0.8>
     Scale Factor=&MakeVector4_FromLinearColor(LinearColor=$Particles.Color)
    )
   ) {isSetParameters=true}
  ]
  rendererClasses [&NiagaraRibbonRendererProperties]
  renderers [
   {rendererClass=&NiagaraRibbonRendererProperties}
   {rendererClass=&NiagaraRibbonRendererProperties, rendererIndex=1}
  ]
 }
}
summary {
 systemName=NS_Pickaxe_NavyArtisan_Trail
 emitters [
  {emitterName=Trail, rendererClasses [&NiagaraRibbonRendererProperties]}
  {emitterName=TrailGrid, rendererClasses [&NiagaraRibbonRendererProperties]}
 ]
 userVariables [
  User.RibbonIDOffset {type=NiagaraInt32}
  User.SwingSectionIndex {type=NiagaraInt32}
 ]
}
